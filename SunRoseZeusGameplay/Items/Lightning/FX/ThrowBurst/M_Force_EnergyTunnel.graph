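# Unlit translucent Niagara mesh-particle material for an energy tunnel in the Zeus lightning item's throw-burst effect.
# Panning noise layers, distorted and randomly time-offset per particle, are eroded and masked by a sine V gradient.
# Dynamic parameters drive erosion, tiling, refraction, optional animated panning and a shift to tinted monochrome.
# TwoSidedSign zeroes front-face opacity, leaving back faces visible with depth, near-camera and axis-aligned Fresnel fades.
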Material /SunRoseZeusGameplay/Items/Lightning/FX/ThrowBurst/M_Force_EnergyTunnel
graph {
 properties {
  bUsedWithNiagaraMeshParticles=true
  blendMode=BLEND_Translucent
  pixelDepthOffsetMode=PDOM_Legacy
  refractionMethod=RM_IndexOfRefraction
  shadingModel=MSM_Unlit
  twoSided=true
  usedShadingModels=Unlit
 }
 nodes [
  AppendVector(A=Constant() {r=1}, B=Constant() {r=1})
  Constant() {r=1}
  Constant() {r=3}
  n1 = TextureCoordinate()
  n2 = NamedRerouteUsage()
  n3 = Constant() {r=1}
  n4 = DynamicParameter() {defaultValue=<rgba 0 0 1 1>, paramNames [Refraction,Erode,AnimPanning,TileScale]}
  n5 = Multiply(A=Constant() {r=0.5}, B=@n4.TileScale)
  n6 = Panner(
   Coordinate=Multiply(A=Add(A=@n1, B=@n2), B=AppendVector(A=@n3, B=@n5))
   Time=NamedRerouteUsage()
   Speed=AppendVector(A=Constant() {r=-0.8}, B=Constant() {r=1})
  ) {speedY=1}
  n7 = Multiply(
   A=TextureSample(UVs=@n6) {samplerType=SAMPLERTYPE_Grayscale, texture=&LowResBlurredNoise}.R
   B=ScalarParameter() {defaultValue=5, parameterName=OpacityMulti}
  )
  n8 = TextureCoordinate()
  n9 = NamedRerouteUsage()
  n10 = Constant() {r=3}
  n11 = Panner(
   Coordinate=Multiply(A=Add(A=@n8, B=@n9), B=AppendVector(A=@n10, B=@n5))
   Time=NamedRerouteUsage()
   Speed=AppendVector(A=Constant() {r=-0.5}, B=Constant() {r=1})
  ) {speedY=1}
  n12 = Multiply(A=Add(A=TextureCoordinate(), B=NamedRerouteUsage()), B=@n4.TileScale)
  n13 = NamedRerouteUsage()
  n14 = AppendVector(A=Constant() {r=-0.7}, B=Constant() {r=0.5})
  n15 = Add(
   A=Add(
    A=TextureSample(UVs=@n11) {texture=/Game/Effects/Fort_Effects/Textures/Tile/T_Noise_03}.RGB
    B=TextureSample(UVs=@n6) {texture=/Game/Effects/Fort_Effects/Textures/Tile/T_Noise_02}.RGB
   )
   B=TextureSample(UVs=Panner(Coordinate=@n12, Time=@n13, Speed=@n14) {speedY=1}) {texture=&T_Noise_03}.RGB
  )
  n16 = DynamicParameter() {defaultValue=<rgba 0 0 1 1>, paramNames [Refraction,Erode,AnimPanning,TileScale]}
  n17 = Subtract(A=@n15, B=@n16.Erode)
  n18 = MaterialFunctionCall() {materialFunction=&LinearGradient}
  n19 = Multiply(A=Sine(None=Multiply(A=@n18.VGradient) {constB=0.5}), B=Constant() {r=1})
  n20 = Constant() {r=2}
  n21 = Multiply(A=Multiply(A=@n7, B=Saturate(None=@n17)), B=Saturate(None=Power(Base=@n19, Exp=@n20)))
  n22 = Saturate(None=MaterialFunctionCall(Distance=Constant() {r=3500}) {materialFunction=&Near_Camera_Fade}.Result)
  LinearInterpolate(Alpha=Multiply(A=Multiply(A=@n21, B=@n16.Refraction), B=LinearInterpolate(Alpha=@n22) {constA=1, constB=0})) {constA=1, constB=1.1}
  MaterialFunctionCall() {materialFunction=&FadeBasedOnViewAngle}
  MaterialFunctionCall() {materialFunction=&DebugTimeSine}
  n23 = Multiply(
   A=LinearInterpolate(
    A=VectorParameter() {parameterName=SubColor, defaultValue=<rgba 0.4803 7.0669 40 0>}.RGB
    B=VectorParameter() {parameterName=MainColor, defaultValue=<rgba 3.75 12.6042 25 0>}.RGB
    Alpha=Saturate(None=@n7)
   )
   B=Desaturation(None=@n15, Fraction=Constant() {r=-0.5})
  )
  n24 = Multiply(A=Desaturation(None=@n23, Fraction=Constant() {r=1}), B=VectorParameter() {defaultValue=<rgba 0.9896 0.7997 0.6251 0>}.RGB)
  n25 = DynamicParameter() {parameterIndex=1, defaultValue=<rgba 0 0 1 1>}
  n26 = ParticleColor()
  MaterialFunctionCall(Emissive=Multiply(A=LinearInterpolate(A=@n23, B=@n24, Alpha=@n25.R), B=@n26.RGB)) {materialFunction=&MF_FX_EmissiveAdjustments}
  n27 = NamedRerouteUsage()
  Multiply(A=@n26.A, B=LinearInterpolate(A=Multiply(A=@n21, B=@n27), Alpha=TwoSidedSign()) {constB=0})
  n28 = Time()
  n29 = DynamicParameter() {defaultValue=<rgba 0 0 1 1>, paramNames [Refraction,Erode,AnimPanning,TileScale]}
  n30 = ScalarParameter() {parameterName=UseAnimPanning}
  NamedRerouteDeclaration(None=Add(A=LinearInterpolate(A=@n28, B=@n29.AnimPanning, Alpha=@n30), B=ParticleRandom())) {name=RandomTimeOffset, nodeColor=<rgba 1 0 0.6328 1>}
  n31 = TextureCoordinate()
  n32 = AppendVector(
   A=ScalarParameter() {defaultValue=5, parameterName=DIstortion_U_Tile}
   B=ScalarParameter() {defaultValue=0.5, parameterName=DIstortion_V_Tile}
  )
  n33 = Constant() {r=-0.5}
  n34 = Constant() {r=0.7}
  n35 = TextureSample(UVs=Panner(Coordinate=Multiply(A=@n31, B=@n32), Time=NamedRerouteUsage(), Speed=AppendVector(A=@n33, B=@n34)) {speedY=1}) {texture=&T_NoisePack_01}
  NamedRerouteDeclaration(None=Divide(A=AppendVector(A=@n35.R, B=@n35.R), B=ScalarParameter() {defaultValue=10, parameterName=Distortion})) {name=Distortion, nodeColor=<rgba 1 0.5156 0 1>}
  n36 = MaterialFunctionCall(Distance=Constant() {r=400}) {materialFunction=&Near_Camera_Fade}
  n37 = MaterialFunctionCall!(
   Axis=Transform(None=Constant3Vector() {constant=<rgba 1 0 0 1>}) {transformSourceType=TRANSFORMSOURCE_Instance}
   Exponent=Constant() {r=1}
   Normal=Multiply(A=VertexNormalWS()) {constB=-1}
  ) {materialFunction=/Game/Effects/Fort_Effects/Materials/_Functions/MF_AxisAlignedFrensnel}
  NamedRerouteDeclaration(None=DepthFade(Opacity=Multiply(A=@n36.Result, B=@n37.Fresnel), FadeDistance=ScalarParameter() {defaultValue=200, parameterName=FadeDistance})) {name=Fade, nodeColor=<rgba 0.3125 0 1 1>}
  TextureSample(UVs=@n6) {samplerType=SAMPLERTYPE_Normal, texture=/Game/Characters/Player/Female/Medium/Bodies/F_MED_Bistro_Astronaut/FX/T_GaussianNoise_Normal}
 ]
}
